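# Looping Sound Cue that plays sporadic wooden barn hinge and joist creaks as a Fortnitemares ambient spot sound.
# A Random node picks one of eight equally weighted creak waves behind a randomized multi-second Delay, which a Looping node repeats.
# A top-level Modulator slightly varies pitch.
# Playback uses distant-creak attenuation, the indoor ambient one-shot sound class and a shared spot concurrency asset.
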SoundCue /Hermes/Audio/Fortnitemares/Spots/Oneshots/SC_Spot_Wooden_Barn_Hinges_Joists_Creaking_Loop_Ext
root=Modulator_3
nodes [
 {id=Delay_0, children [Random_1], properties {delayMax=30, delayMin=9}}
 {id=Looping_0, children [Delay_0]}
 {id=Modulator_3, children [Looping_0], properties {pitchMax=1.1, pitchMin=0.9}}
 {
  id=Random_1
  children [WavePlayer_5, WavePlayer_9, WavePlayer_10, WavePlayer_11, WavePlayer_12, WavePlayer_13, WavePlayer_14, WavePlayer_15]
  properties {weights [1,1,1,1,1,1,1,1]}
 }
 {id=WavePlayer_10, properties {soundWaveAssetPtr=&/Spot_Wooden_Barn_Hinges_Joists_Creaking_03}}
 {id=WavePlayer_11, properties {soundWaveAssetPtr=&/Spot_Wooden_Barn_Hinges_Joists_Creaking_04}}
 {id=WavePlayer_12, properties {soundWaveAssetPtr=&/Spot_Wooden_Barn_Hinges_Joists_Creaking_05}}
 {id=WavePlayer_13, properties {soundWaveAssetPtr=&/Spot_Wooden_Barn_Hinges_Joists_Creaking_06}}
 {id=WavePlayer_14, properties {soundWaveAssetPtr=&/Spot_Wooden_Barn_Hinges_Joists_Creaking_07}}
 {id=WavePlayer_15, properties {soundWaveAssetPtr=&/Spot_Wooden_Barn_Hinges_Joists_Creaking_08}}
 {id=WavePlayer_5, properties {soundWaveAssetPtr=&/Spot_Wooden_Barn_Hinges_Joists_Creaking_01}}
 {id=WavePlayer_9, properties {soundWaveAssetPtr=&/Spot_Wooden_Barn_Hinges_Joists_Creaking_02}}
]
properties {
 attenuationSettings=/Game/Sounds/Attenuation/Fort_TileAmbient_OneShot_Distant_Creak_Att
 bEnableBusSends=false
 maxDistance=8200
 soundClassObject=/Game/Sounds/FortSoundClassesAndModes/Fort_Ambient_OneShots_Indoors
 volumeMultiplier=1
 concurrencyOverrides {maxCount=3, retriggerTime=2, voiceStealReleaseTime=2, volumeScaleAttackTime=1}
 concurrencySet [/Hermes/Props/SprocketPoppy/Materials/FaceAnim/AmbientConc_3Max_Spot_Concurrency]
 preEffectBusSends [{}]
}
