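# Fortnitemares ambient spot one-shot SoundCue for exterior creaking and stressing of a wooden building's wall frame.
# A Random node picks one of ten equally weighted wave variations, then a Modulator applies slight pitch variation.
# It plays in the outdoor ambient one-shot sound class with a distant-creak attenuation.
# An assigned concurrency set, plus override values in the cue, relates to capping simultaneous instances.
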SoundCue /Hermes/Audio/Fortnitemares/Spots/Oneshots/SC_Spot_Wooden_Building_Stressing_Wall_Frame_Creaking_Movement_Ext
root=Modulator_3
nodes [
 {id=Modulator_3, children [Random_0], properties {pitchMax=1.1, pitchMin=0.9}}
 {
  id=Random_0
  children [WavePlayer_0, WavePlayer_1, WavePlayer_2, WavePlayer_3, WavePlayer_4, WavePlayer_6, WavePlayer_7, WavePlayer_8, WavePlayer_22, WavePlayer_23]
  properties {weights [1,1,1,1,1,1,1,1,1,1]}
 }
 {
  id=WavePlayer_0
  properties {soundWaveAssetPtr=&/Spot_Wooden_Building_Stressing_Wall_Frame_Creaking_Movement_01}
 }
 {
  id=WavePlayer_1
  properties {soundWaveAssetPtr=&/Spot_Wooden_Building_Stressing_Wall_Frame_Creaking_Movement_02}
 }
 {
  id=WavePlayer_2
  properties {soundWaveAssetPtr=&/Spot_Wooden_Building_Stressing_Wall_Frame_Creaking_Movement_03}
 }
 {
  id=WavePlayer_22
  properties {soundWaveAssetPtr=&/Spot_Wooden_Building_Stressing_Wall_Frame_Creaking_Movement_09}
 }
 {
  id=WavePlayer_23
  properties {soundWaveAssetPtr=&/Spot_Wooden_Building_Stressing_Wall_Frame_Creaking_Movement_10}
 }
 {
  id=WavePlayer_3
  properties {soundWaveAssetPtr=&/Spot_Wooden_Building_Stressing_Wall_Frame_Creaking_Movement_04}
 }
 {
  id=WavePlayer_4
  properties {soundWaveAssetPtr=&/Spot_Wooden_Building_Stressing_Wall_Frame_Creaking_Movement_05}
 }
 {
  id=WavePlayer_6
  properties {soundWaveAssetPtr=&/Spot_Wooden_Building_Stressing_Wall_Frame_Creaking_Movement_06}
 }
 {
  id=WavePlayer_7
  properties {soundWaveAssetPtr=&/Spot_Wooden_Building_Stressing_Wall_Frame_Creaking_Movement_07}
 }
 {
  id=WavePlayer_8
  properties {soundWaveAssetPtr=&/Spot_Wooden_Building_Stressing_Wall_Frame_Creaking_Movement_08}
 }
]
properties {
 attenuationSettings=/Game/Sounds/Attenuation/Fort_TileAmbient_OneShot_Distant_Creak_Att
 bEnableBusSends=false
 maxDistance=8200
 soundClassObject=&Fort_Ambient_OneShots_Outdoors
 volumeMultiplier=1
 concurrencyOverrides {maxCount=3, retriggerTime=2, voiceStealReleaseTime=2, volumeScaleAttackTime=1}
 concurrencySet [/Hermes/Props/SprocketPoppy/Materials/FaceAnim/AmbientConc_3Max_Spot_Concurrency]
 preEffectBusSends [{}]
}
