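# Sound cue for ominous building creak one-shots, part of the Fortnitemares ambient spot sounds, routed to the indoor ambient one-shot sound class.
# A random node picks one of eight equally weighted creak waves and feeds a modulator that slightly varies pitch.
# All four quality levels share that same chain.
# It uses ambient attenuation and overrides concurrency to cap simultaneous voices and limit retriggering.
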SoundCue /Hermes/Audio/Fortnitemares/Spots/Oneshots/Spot_Ominious_Building_Creak_Cue
root=QualityLevel_0
nodes [
 {id=Modulator_3, children [Random_0], properties {pitchMax=1.1, pitchMin=0.9}}
 {id=QualityLevel_0, children [Modulator_3,Modulator_3,Modulator_3,Modulator_3]}
 {
  id=Random_0
  children [WavePlayer_0,WavePlayer_1,WavePlayer_2,WavePlayer_3,WavePlayer_4,WavePlayer_6,WavePlayer_7,WavePlayer_8]
  properties {weights [1,1,1,1,1,1,1,1]}
 }
 {id=WavePlayer_0, properties {soundWaveAssetPtr=&/Spot_Ominious_Building_Creak_01}}
 {id=WavePlayer_1, properties {soundWaveAssetPtr=&/Spot_Ominious_Building_Creak_02}}
 {id=WavePlayer_2, properties {soundWaveAssetPtr=&/Spot_Ominious_Building_Creak_03}}
 {id=WavePlayer_3, properties {soundWaveAssetPtr=&/Spot_Ominious_Building_Creak_04}}
 {id=WavePlayer_4, properties {soundWaveAssetPtr=&/Spot_Ominious_Building_Creak_05}}
 {id=WavePlayer_6, properties {soundWaveAssetPtr=&/Spot_Ominious_Building_Creak_06}}
 {id=WavePlayer_7, properties {soundWaveAssetPtr=&/Spot_Ominious_Building_Creak_07}}
 {id=WavePlayer_8, properties {soundWaveAssetPtr=&/Spot_Ominious_Building_Creak_08}}
]
properties {
 attenuationSettings=/AmbientAudioBase/Attenuations/Fort_Ambient_Athena_Att
 bEnableBusSends=false
 bOverrideConcurrency=true
 maxDistance=2097152
 soundClassObject=/Game/Sounds/FortSoundClassesAndModes/Fort_Ambient_OneShots_Indoors
 volumeMultiplier=0.8
 concurrencyOverrides {maxCount=3, retriggerTime=2, voiceStealReleaseTime=2, volumeScaleAttackTime=1}
 preEffectBusSends [{}]
}
