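# Sound cue for a Fortnitemares ambient one-shot of organic growth chewing, built from eleven pumpkin vine growth movement waves.
# A quality-level node routes to two duplicate branches, each randomly choosing between a five-wave and six-wave group, then a wave within it.
# Each branch applies identical random pitch and volume modulation, so all quality levels share the same variations.
# It uses distant-creak attenuation, the indoor ambient one-shot sound class and a tight concurrency override.
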SoundCue /Hermes/Audio/Fortnitemares/Spots/Oneshots/Spot_Organic_Growth_Chew_Cue
root=QualityLevel_0
nodes [
 {id=Modulator_0, children [Random_5], properties {pitchMax=1.1, pitchMin=0.75, volumeMax=1.35, volumeMin=1.25}}
 {id=Modulator_3, children [Random_3], properties {pitchMax=1.1, pitchMin=0.75, volumeMax=1.35, volumeMin=1.25}}
 {id=QualityLevel_0, children [Modulator_3,Modulator_3,Modulator_0,Modulator_0]}
 {id=Random_0, children [WavePlayer_5, WavePlayer_9, WavePlayer_10, WavePlayer_11, WavePlayer_15], properties {weights [1,1,1,1,1]}}
 {id=Random_1, children [WavePlayer_0,WavePlayer_1,WavePlayer_2,WavePlayer_3,WavePlayer_4], properties {weights [1,1,1,1,1]}}
 {
  id=Random_2
  children [WavePlayer_6, WavePlayer_7, WavePlayer_8, WavePlayer_12, WavePlayer_13, WavePlayer_14]
  properties {weights [1,1,1,1,1,1]}
 }
 {id=Random_3, children [Random_1,Random_2], properties {weights [1,1]}}
 {
  id=Random_4
  children [WavePlayer_16, WavePlayer_17, WavePlayer_18, WavePlayer_19, WavePlayer_20, WavePlayer_21]
  properties {weights [1,1,1,1,1,1]}
 }
 {id=Random_5, children [Random_0,Random_4], properties {weights [1,1]}}
 {id=WavePlayer_0, properties {soundWaveAssetPtr=&/Spot_Pumpkin_Vine_Growth_Movement_01}}
 {id=WavePlayer_1, properties {soundWaveAssetPtr=&/Spot_Pumpkin_Vine_Growth_Movement_02}}
 {id=WavePlayer_10, properties {soundWaveAssetPtr=&/Spot_Pumpkin_Vine_Growth_Movement_03}}
 {id=WavePlayer_11, properties {soundWaveAssetPtr=&/Spot_Pumpkin_Vine_Growth_Movement_04}}
 {id=WavePlayer_12, properties {soundWaveAssetPtr=&/Spot_Pumpkin_Vine_Growth_Movement_09}}
 {id=WavePlayer_13, properties {soundWaveAssetPtr=&/Spot_Pumpkin_Vine_Growth_Movement_10}}
 {id=WavePlayer_14, properties {soundWaveAssetPtr=&/Spot_Pumpkin_Vine_Growth_Movement_11}}
 {id=WavePlayer_15, properties {soundWaveAssetPtr=&/Spot_Pumpkin_Vine_Growth_Movement_05}}
 {id=WavePlayer_16, properties {soundWaveAssetPtr=&/Spot_Pumpkin_Vine_Growth_Movement_06}}
 {id=WavePlayer_17, properties {soundWaveAssetPtr=&/Spot_Pumpkin_Vine_Growth_Movement_07}}
 {id=WavePlayer_18, properties {soundWaveAssetPtr=&/Spot_Pumpkin_Vine_Growth_Movement_08}}
 {id=WavePlayer_19, properties {soundWaveAssetPtr=&Spot_Pumpkin_Vine_Growth_Movement_09}}
 {id=WavePlayer_2, properties {soundWaveAssetPtr=&Spot_Pumpkin_Vine_Growth_Movement_03}}
 {id=WavePlayer_20, properties {soundWaveAssetPtr=&Spot_Pumpkin_Vine_Growth_Movement_10}}
 {id=WavePlayer_21, properties {soundWaveAssetPtr=&Spot_Pumpkin_Vine_Growth_Movement_11}}
 {id=WavePlayer_3, properties {soundWaveAssetPtr=&Spot_Pumpkin_Vine_Growth_Movement_04}}
 {id=WavePlayer_4, properties {soundWaveAssetPtr=&Spot_Pumpkin_Vine_Growth_Movement_05}}
 {id=WavePlayer_5, properties {soundWaveAssetPtr=&Spot_Pumpkin_Vine_Growth_Movement_01}}
 {id=WavePlayer_6, properties {soundWaveAssetPtr=&Spot_Pumpkin_Vine_Growth_Movement_06}}
 {id=WavePlayer_7, properties {soundWaveAssetPtr=&Spot_Pumpkin_Vine_Growth_Movement_07}}
 {id=WavePlayer_8, properties {soundWaveAssetPtr=&Spot_Pumpkin_Vine_Growth_Movement_08}}
 {id=WavePlayer_9, properties {soundWaveAssetPtr=&Spot_Pumpkin_Vine_Growth_Movement_02}}
]
properties {
 attenuationSettings=/Game/Sounds/Attenuation/Fort_TileAmbient_OneShot_Distant_Creak_Linear_Att
 bEnableBusSends=false
 bOverrideConcurrency=true
 maxDistance=8700
 soundClassObject=/Game/Sounds/FortSoundClassesAndModes/Fort_Ambient_OneShots_Indoors
 volumeMultiplier=0.35
 concurrencyOverrides {maxCount=2, voiceStealReleaseTime=0.25}
 preEffectBusSends [{}]
}
