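# Sound cue for a close vine creak ambient one-shot spot in Fortnitemares, on the indoor ambient one-shots sound class with a distant-creak attenuation asset.
# Nested random nodes pick one of two pairs, then one of four creak waves, and a modulator adds slight pitch variation.
# All quality levels share that same branch.
# A concurrency override limits simultaneous voices, with a short release on voice stealing.
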SoundCue /Hermes/Audio/Fortnitemares/Spots/Oneshots/Spot_Organic_Vine_Creak_Close_Cue
root=QualityLevel_0
nodes [
 {id=Modulator_3, children [Random_3], properties {pitchMax=1.1, pitchMin=0.9}}
 {id=QualityLevel_0, children [Modulator_3,Modulator_3,Modulator_3,Modulator_3]}
 {id=Random_1, children [WavePlayer_5,WavePlayer_9], properties {weights [1,1]}}
 {id=Random_2, children [WavePlayer_10, WavePlayer_11], properties {weights [1,1]}}
 {id=Random_3, children [Random_1,Random_2], properties {weights [1,1]}}
 {id=WavePlayer_10, properties {soundWaveAssetPtr=&/Spot_Organic_Vine_Creak_Close_03}}
 {id=WavePlayer_11, properties {soundWaveAssetPtr=&/Spot_Organic_Vine_Creak_Close_04}}
 {id=WavePlayer_5, properties {soundWaveAssetPtr=&/Spot_Organic_Vine_Creak_Close_01}}
 {id=WavePlayer_9, properties {soundWaveAssetPtr=&/Spot_Organic_Vine_Creak_Close_02}}
]
properties {
 attenuationSettings=/Game/Sounds/Attenuation/Fort_TileAmbient_OneShot_Distant_Creak_Linear_Att
 bEnableBusSends=false
 bOverrideConcurrency=true
 maxDistance=8700
 soundClassObject=/Game/Sounds/FortSoundClassesAndModes/Fort_Ambient_OneShots_Indoors
 volumeMultiplier=1
 concurrencyOverrides {maxCount=2, voiceStealReleaseTime=0.25}
 preEffectBusSends [{}]
}
