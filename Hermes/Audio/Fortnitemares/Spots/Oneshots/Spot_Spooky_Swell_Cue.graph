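# Fortnitemares ambient one-shot Sound Cue that plays a random spooky swell, routed to the outdoor one-shots sound class with shared Athena ambient attenuation.
# A quality-level root sends its first two slots to a nine-wave random pool with pitch and wide volume modulation.
# Its last two slots use a smaller three-wave pool with pitch modulation, trimming variation by quality level.
# A concurrency override caps simultaneous voices and sets a retrigger time.
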SoundCue /Hermes/Audio/Fortnitemares/Spots/Oneshots/Spot_Spooky_Swell_Cue
root=QualityLevel_0
nodes [
 {id=Modulator_0, children [Random_2], properties {pitchMax=1.1, pitchMin=0.9}}
 {id=Modulator_3, children [Random_1], properties {pitchMax=1.1, pitchMin=0.9, volumeMin=0.5}}
 {id=QualityLevel_0, children [Modulator_3,Modulator_3,Modulator_0,Modulator_0]}
 {
  id=Random_1
  children [WavePlayer_5, WavePlayer_9, WavePlayer_10, WavePlayer_11, WavePlayer_12, WavePlayer_13, WavePlayer_14, WavePlayer_15, WavePlayer_16]
  properties {weights [1,1,1,1,1,1,1,1,1]}
 }
 {id=Random_2, children [WavePlayer_17, WavePlayer_20, WavePlayer_19], properties {weights [1,1,1]}}
 {id=WavePlayer_10, properties {soundWaveAssetPtr=&/Spot_Spooky_Swell_07}}
 {id=WavePlayer_11, properties {soundWaveAssetPtr=&/Spot_Spooky_Swell_06}}
 {id=WavePlayer_12, properties {soundWaveAssetPtr=&/Spot_Spooky_Swell_05}}
 {id=WavePlayer_13, properties {soundWaveAssetPtr=&/Spot_Spooky_Swell_04}}
 {id=WavePlayer_14, properties {soundWaveAssetPtr=&/Spot_Spooky_Swell_03}}
 {id=WavePlayer_15, properties {soundWaveAssetPtr=&/Spot_Spooky_Swell_02}}
 {id=WavePlayer_16, properties {soundWaveAssetPtr=&/Spot_Spooky_Swell_01}}
 {id=WavePlayer_17, properties {soundWaveAssetPtr=&Spot_Spooky_Swell_03}}
 {id=WavePlayer_19, properties {soundWaveAssetPtr=&Spot_Spooky_Swell_01}}
 {id=WavePlayer_20, properties {soundWaveAssetPtr=&/Spot_Spooky_Swell_09}}
 {id=WavePlayer_5, properties {soundWaveAssetPtr=&Spot_Spooky_Swell_09}}
 {id=WavePlayer_9, properties {soundWaveAssetPtr=&/Spot_Spooky_Swell_08}}
]
properties {
 attenuationSettings=/AmbientAudioBase/Attenuations/Fort_Ambient_Athena_Att
 bEnableBusSends=false
 bOverrideConcurrency=true
 maxDistance=2097152
 soundClassObject=&Fort_Ambient_OneShots_Outdoors
 volumeMultiplier=0.5
 concurrencyOverrides {maxCount=3, retriggerTime=2, voiceStealReleaseTime=2, volumeScaleAttackTime=1}
 preEffectBusSends [{}]
}
